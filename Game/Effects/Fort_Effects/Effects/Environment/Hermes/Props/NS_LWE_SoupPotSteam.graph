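# NS_LWE_SoupPotSteam is a Niagara system for steam from a soup pot prop, filed under Hermes environment effects.
# It holds one emitter, Steam_SoupPot, with no renderers listed, and exposes no user parameters.
# The system uses SystemState and a spawn-time set-variables assignment.
# Its dynamic inputs include random-range floats, float multiplication and a float comparison that yields a bool, plus a bool-to-custom-float conversion.
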NiagaraSystem /Game/Effects/Fort_Effects/Effects/Environment/Hermes/Props/NS_LWE_SoupPotSteam
dependencies {
 usedDataInterfaces []
 usedDynamicInputs [
  &Multiply_Float
  &MakeCustomFloatFromBool
  &SetBoolByFloatComparison
  &RandomRangeFloat
 ]
 usedModules [
  &self:SystemSpawnScript.SystemScriptSource.SystemScriptGraph.NiagaraNodeAssignment_0.SetVariables_#1
  &SystemState
 ]
}
emitters {Steam_SoupPot {rendererClasses [], renderers []}}
summary {systemName=NS_LWE_SoupPotSteam, emitters [{emitterName=Steam_SoupPot, rendererClasses []}], userVariables []}
